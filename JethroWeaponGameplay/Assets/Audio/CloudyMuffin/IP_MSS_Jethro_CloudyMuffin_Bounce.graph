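# MetaSound Source preset that plays the bounce sound for the Jethro CloudyMuffin weapon, as a mono one-shot.
# It reuses the MSS_Weapon_Equip parent graph, supplying eight bounce wave variations and random pitch-shift settings.
# Grenade bounce attenuation and the ranged-projectile effects sound class place it in the mix.
# A concurrency override caps simultaneous voices and limits rapid retriggering.
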MetaSoundSource /JethroWeaponGameplay/Assets/Audio/CloudyMuffin/IP_MSS_Jethro_CloudyMuffin_Bounce
preset_parent=/FortMetasound/Weapons/MSS_Weapon_Equip
inputs [
 Character.Affiliation {type=Enum:FortAudioAffiliation}
 Character.Health {type=Float}
 Character.UniqueID {type=Int32}
 MasterVolume {type=Float}
 Random.PitchShift.Base {type=Float}
 Random.PitchShift.Spread {type=Float}
 Sounds {type=WaveAsset:Array}
 UE.Source.OnPlay {type=Trigger}
]
interfaces [Fort.Character:1.0, UE.OutputFormat.Mono:1.0, UE.Source.OneShot:1.0, UE.Source:1.1]
outputs [
 "UE.OutputFormat.Mono.Audio:0" {type=Audio}
 UE.Source.OneShot.OnFinished {type=Trigger}
]
pages [Default {input_defaults {
   Character.Affiliation=0
   Character.Health=0
   Character.UniqueID=0
   MasterVolume=1
   Random.PitchShift.Base=0
   Random.PitchShift.Spread=0.5
   UE.Source.OnPlay=null
   Sounds [
    &/IP_Jethro_CloudyMuffin_Bounce_01
    &/IP_Jethro_CloudyMuffin_Bounce_02
    &/IP_Jethro_CloudyMuffin_Bounce_03
    &/IP_Jethro_CloudyMuffin_Bounce_04
    &/IP_Jethro_CloudyMuffin_Bounce_05
    &/IP_Jethro_CloudyMuffin_Bounce_06
    &/IP_Jethro_CloudyMuffin_Bounce_07
    &/IP_Jethro_CloudyMuffin_Bounce_08
   ]
  }}]
properties {
 assetImportData=&self:AssetImportData
 attenuationSettings=/Game/Sounds/Attenuation/Fort_Grenade_Bounce_Att
 bOverrideConcurrency=true
 maxDistance=5500
 sampleRate=48000
 soundClassObject=&Fort_WeaponRanged_Projectile_Effects
 concurrencyOverrides {maxCount=3, retriggerTime=0.1}
}
