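# Mono one-shot MetaSound Source for the pre-explode sound of the Jethro CloudyMuffin weapon, routed to the explosives sound class.
# A wave player starts the exposed Sound asset on play, and its output is scaled by the MasterVolume input.
# The wave player's finish trigger ends the source.
# The source uses the CloudyMuffin fuse attenuation and overrides concurrency to limit simultaneous instances.
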MetaSoundSource /JethroWeaponGameplay/Assets/Audio/CloudyMuffin/IP_MSS_Jethro_CloudyMuffin_PreExplode
inputs [
 MasterVolume {type=Float}
 Sound {type=WaveAsset}
 UE.Source.OnPlay {type=Trigger}
]
interfaces [UE.OutputFormat.Mono:1.0, UE.Source.OneShot:1.0, UE.Source:1.1]
outputs [
 "UE.OutputFormat.Mono.Audio:0" {type=Audio}
 UE.Source.OneShot.OnFinished {type=Trigger}
]
pages [Default {
  input_defaults {MasterVolume=1, Sound=&/IP_Jethro_CloudyMuffin_PreExplode_01, UE.Source.OnPlay=null}
  nodes [
   WavePlayer_0 = UE.Wave Player:Mono(Wave Asset=^Sound, Play=^UE.Source.OnPlay)
   Multiply_0 = UE.Multiply:Audio by Float(AdditionalOperands=^MasterVolume, PrimaryOperand=@WavePlayer_0.Out Mono)
  ]
  links [{dst="@UE.OutputFormat.Mono.Audio:0", src=Multiply_0.Out}, {dst="@UE.Source.OneShot.OnFinished", src=WavePlayer_0.On Finished}]
 }]
properties {
 assetImportData=&self:AssetImportData
 attenuationSettings=&/IP_Att_Jethro_CloudyMuffin_Fuse
 bOverrideConcurrency=true
 maxDistance=5250
 sampleRate=48000
 soundClassObject=/Game/Sounds/FortSoundClassesAndModes/Fort_Explosives
 volume=0.7
 concurrencyOverrides {maxCount=2}
}
