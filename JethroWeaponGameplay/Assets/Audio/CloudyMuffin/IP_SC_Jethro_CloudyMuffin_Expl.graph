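# Explosion SoundCue for the Jethro CloudyMuffin weapon, built from close, medium and distant layers mixed together.
# Each layer randomly picks one of four waves, varies it through a modulator, then applies its own attenuation asset and sound class override.
# The distant layer adds a short random delay and lowered pitch, sits under a distant weapon-fire class, and mixes quieter alongside the medium layer.
# Concurrency overrides cap simultaneous instances.
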SoundCue /JethroWeaponGameplay/Assets/Audio/CloudyMuffin/IP_SC_Jethro_CloudyMuffin_Expl
root=Mixer_5
nodes [
 {
  id=Attenuation_0
  children [Modulator_3]
  properties {
   attenuationSettings=&/IP_ATT_Jethro_CloudyMuffin_Expl_Close
   attenuationOverrides {falloffDistance=900, attenuationShapeExtents {x=100}}
  }
 }
 {
  id=Attenuation_4
  children [Modulator_0]
  properties {attenuationSettings=&/IP_ATT_Jethro_CloudyMuffin_Expl_Medium}
 }
 {
  id=Attenuation_8
  children [Delay_0]
  properties {
   attenuationSettings=&/IP_ATT_Jethro_CloudyMuffin_Expl_Distant
   attenuationOverrides {falloffDistance=3000, attenuationShapeExtents {x=1000}}
  }
 }
 {id=Delay_0, children [Modulator_7], properties {delayMax=0.1}}
 {id=Mixer_5, children [SoundClass_3,SoundClass_0,SoundClass_1], properties {inputVolume [1,0.6,0.6]}}
 {id=Modulator_0, children [Random_1]}
 {id=Modulator_3, children [Random_0], properties {pitchMax=1.1, pitchMin=0.9, volumeMax=1, volumeMin=1}}
 {id=Modulator_7, children [Random_2], properties {pitchMax=0.9, pitchMin=0.8}}
 {id=Random_0, children [WavePlayer_0,WavePlayer_1,WavePlayer_2,WavePlayer_3], properties {weights [1,1,1,1]}}
 {id=Random_1, children [WavePlayer_4,WavePlayer_5,WavePlayer_6,WavePlayer_7], properties {weights [1,1,1,1]}}
 {id=Random_2, children [WavePlayer_8, WavePlayer_9, WavePlayer_10, WavePlayer_11], properties {weights [1,1,1,1]}}
 {id=SoundClass_0, children [Attenuation_4], properties {soundClassOverride=/Game/Sounds/FortSoundClassesAndModes/Fort_Explosives_Medium_PSM}}
 {
  id=SoundClass_1
  children [Attenuation_8]
  properties {soundClassOverride=&Fort_WeaponRanged_Fire_Distant}
 }
 {id=SoundClass_3, children [Attenuation_0], properties {soundClassOverride=&Fort_Explosives_Medium_PSM}}
 {id=WavePlayer_0, properties {soundWaveAssetPtr=&/IP_Jethro_CloudyMuffin_Expl_Close_01}}
 {id=WavePlayer_1, properties {soundWaveAssetPtr=&/IP_Jethro_CloudyMuffin_Expl_Close_02}}
 {id=WavePlayer_10, properties {soundWaveAssetPtr=&/IP_Jethro_CloudyMuffin_Expl_Distant_02}}
 {id=WavePlayer_11, properties {soundWaveAssetPtr=&/IP_Jethro_CloudyMuffin_Expl_Distant_01}}
 {id=WavePlayer_2, properties {soundWaveAssetPtr=&/IP_Jethro_CloudyMuffin_Expl_Close_03}}
 {id=WavePlayer_3, properties {soundWaveAssetPtr=&/IP_Jethro_CloudyMuffin_Expl_Close_04}}
 {id=WavePlayer_4, properties {soundWaveAssetPtr=&/IP_Jethro_CloudyMuffin_Expl_Medium_01}}
 {id=WavePlayer_5, properties {soundWaveAssetPtr=&/IP_Jethro_CloudyMuffin_Expl_Medium_02}}
 {id=WavePlayer_6, properties {soundWaveAssetPtr=&/IP_Jethro_CloudyMuffin_Expl_Medium_03}}
 {id=WavePlayer_7, properties {soundWaveAssetPtr=&/IP_Jethro_CloudyMuffin_Expl_Medium_04}}
 {id=WavePlayer_8, properties {soundWaveAssetPtr=&/IP_Jethro_CloudyMuffin_Expl_Distant_04}}
 {id=WavePlayer_9, properties {soundWaveAssetPtr=&/IP_Jethro_CloudyMuffin_Expl_Distant_03}}
]
properties {
 bOverrideConcurrency=true
 maxDistance=27500
 volumeMultiplier=1.25
 attenuationOverrides {reverbWetLevelMax=0.7, reverbWetLevelMin=0.2}
 concurrencyOverrides {maxCount=3, retriggerTime=0.1, volumeScale=0.5}
}
