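# Sound cue for the Jethro CloudyMuffin weapon's arming sound and burning fuse loop, classed as an explosives sound.
# A mixer plays a one-shot arm wave alongside a looping fuse wave at slightly lower input volume.
# The fuse loop starts after a short fixed delay, and an enveloper ramps its volume up from silence.
# A fuse attenuation asset handles spatialization, and a concurrency override caps simultaneous instances.
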SoundCue /JethroWeaponGameplay/Assets/Audio/CloudyMuffin/IP_SC_Jethro_CloudyMuffin_Fuse_Loop
root=Mixer_0
nodes [
 {id=Delay_0, children [WavePlayer_1], properties {delayMax=0.35, delayMin=0.35}}
 {
  id=Enveloper_2
  children [Delay_0]
  properties {pitchCurve {editorCurveData {keys [{time=0}]}}, volumeCurve {editorCurveData {keys [<at 0 0>,<at 0.5632 0.191>,<at 1 1>]}}}
 }
 {id=Mixer_0, children [WavePlayer_2,Enveloper_2], properties {inputVolume [1,0.7]}}
 {
  id=WavePlayer_1
  properties {bLooping=true, soundWaveAssetPtr=&/IP_Jethro_CloudyMuffin_Fuse_01}
 }
 {id=WavePlayer_2, properties {soundWaveAssetPtr=&/IP_Jethro_CloudyMuffin_Arm_01}}
]
properties {
 attenuationSettings=&/IP_Att_Jethro_CloudyMuffin_Fuse
 bOverrideConcurrency=true
 maxDistance=5250
 soundClassObject=/Game/Sounds/FortSoundClassesAndModes/Fort_Explosives
 volumeMultiplier=0.4
 attenuationOverrides {distanceAlgorithm=Logarithmic, falloffDistance=1200, attenuationShapeExtents {x=500}}
 concurrencyOverrides {maxCount=2}
}
